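# Translucent unlit Niagara sprite and mesh material for chromatic hand energy in the CampWrench emote.
# Panning noise distorts U-derived flame-texture UVs, shifted by dynamic OffsetX, and varies the offset between three samples packed into RGB for color fringing.
# Refine parameters and particle color shape emissive.
# Flame, smoke mask, V gradient and particle alpha form a depth-faded mask that drives an index-of-refraction value, simplified by quality and shading path.
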Material /BRCosmetics/Animation/Game/MainPlayer/Emotes/CampWrench/FX/Materials/M_Campwrench_ChromaHandEnergy
graph {
 properties {
  bUseTranslucencyVertexFog=false
  bUsedWithNiagaraMeshParticles=true
  bUsedWithNiagaraSprites=true
  blendMode=BLEND_Translucent
  pixelDepthOffsetMode=PDOM_Legacy
  refractionMethod=RM_IndexOfRefraction
  shadingModel=MSM_Unlit
  usedShadingModels=Unlit
 }
 nodes [
  n1 = TextureSample!(UVs=Panner(Coordinate=TextureCoordinate() {uTiling=2}) {speedY=-0.2}, Tex, Level, Apply View MipBias) {mipValueMode=TMVM_MipLevel, samplerType=SAMPLERTYPE_Grayscale, texture=&LowResBlurredNoise}
  n2 = Multiply(A=@n1.R, B=Constant2Vector() {g=0.05, r=0.025})
  n3 = MaterialFunctionCall(Float2=TextureCoordinate()) {materialFunction=&BreakOutFloat2Components}
  n4 = AppendVector(A=@n3.R, B=@n3.R)
  n5 = Constant2Vector() {g=0.2, r=1}
  n6 = DynamicParameter() {paramNames [OffsetX,OffsetY,Param3,Param4]}
  n7 = Add(A=@n2, B=Add(A=Multiply(A=@n4, B=@n5), B=AppendVector(A=@n6.OffsetX, B=@n6.OffsetX)))
  n8 = LinearInterpolate(
   A=ScalarParameter() {defaultValue=0.005, parameterName=OffsetChannel1}
   B=ScalarParameter() {defaultValue=0.02, parameterName=OffsetChannel2}
   Alpha=@n1.R
  )
  n9 = MaterialFunctionCall(
   X=TextureSample!(UVs=Add(A=@n7, B=@n8), Tex, Apply View MipBias) {samplerType=SAMPLERTYPE_LinearGrayscale, texture=/BRCosmetics/Weapons/FORT_Melee/Pickaxe_Cube_Ninja_Male/FX/T_FluidFlameTile}.R
   Y=TextureSample(UVs=@n7) {samplerType=SAMPLERTYPE_LinearGrayscale, texture=&T_FluidFlameTile}.R
   Z=TextureSample(UVs=Subtract(A=@n7, B=@n8)) {samplerType=SAMPLERTYPE_LinearGrayscale, texture=&T_FluidFlameTile}.R
  ) {materialFunction=&MakeFloat3}
  n10 = VectorParameter() {parameterName=Refine, defaultValue=<rgba 1 1 1 1>}
  n11 = Add(A=Multiply(A=Power(Base=@n9.Result, Exp=@n10.R) {constExponent=6}, B=@n10.G) {constB=100}, B=Multiply(A=@n9.Result, B=@n10.B) {constB=5})
  n12 = Constant() {r=-0.5}
  n13 = ParticleColor()
  MaterialFunctionCall(Emissive=Multiply(A=Desaturation(None=@n11, Fraction=@n12), B=@n13.RGB)) {materialFunction=&MF_FX_EmissiveAdjustments}
  n14 = Multiply(A=Desaturation(None=@n9.Result, Fraction=Constant() {r=1}), B=@n10.A) {constB=10}
  Multiply(B=@n14)
  n15 = TextureCoordinate()
  n16 = ComponentMask(None=@n15) {g=true}
  n17 = TextureSample(UVs=Add(A=@n2, B=@n15)) {samplerType=SAMPLERTYPE_Grayscale, texture=/Game/Effects/Fort_Effects/Textures/Smoke/T_Smoke_Trail_Soft_NoClamp}
  n18 = Multiply(A=@n13.A, B=Multiply(A=@n14, B=Multiply(A=@n16, B=@n17.R)))
  n19 = DepthFade(Opacity=@n18) {fadeDistanceDefault=20}
  n20 = LinearInterpolate(Alpha=ShadingPathSwitch(Default=@n19, Deferred=@n19, Forward=@n18, Mobile=@n18)) {constA=1, constB=1.05}
  n21 = Constant() {r=1}
  n22 = QualitySwitch(Default=@n20, Low=@n21, High=@n20, Medium=@n21, Epic=@n20)
  ShadingPathSwitch(Default=@n22, Deferred=@n22, Forward=@n21, Mobile=@n21)
 ]
}
