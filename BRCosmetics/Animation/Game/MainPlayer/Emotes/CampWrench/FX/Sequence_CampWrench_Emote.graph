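# Level Sequence that plays the CampWrench emote on a skeletal mesh actor with its Niagara effects, apparently for authoring or previewing the emote FX.
# The possessed actor holds a fixed transform and runs one skeletal animation section from the start.
# Two child Niagara components use system spawn tracks: the main effect starts with the animation, and the shockwave starts partway through, overlapping both and ending last.
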LevelSequence /BRCosmetics/Animation/Game/MainPlayer/Emotes/CampWrench/FX/Sequence_CampWrench_Emote
bindings [
 NS_Emote_CampWrench_Shockwave {
  id=#1
  parent_id=#2
  possessed_class=NiagaraComponent
  tracks [{type=MovieSceneNiagaraSystemTrack, sections [{type=MovieSceneNiagaraSystemSpawnSection, end_frame=455, start_frame=155}]}]
 }
 NS_Emote_CampWrench {
  id=#3
  parent_id=#2
  possessed_class=NiagaraComponent
  tracks [{type=MovieSceneNiagaraSystemTrack, sections [{type=MovieSceneNiagaraSystemSpawnSection, end_frame=300, start_frame=0}]}]
 }
 Emote_CampWrench_CMF_M {
  id=#2
  possessed_class=FortSkeletalMeshActor
  tracks [
   {
    type=MovieScene3DTransformTrack
    sections [{
      type=MovieScene3DTransformSection
      channels [
       Location.X {type=MovieSceneScriptingDoubleChannel, default=38.1995}
       Location.Y {type=MovieSceneScriptingDoubleChannel, default=1700.5007}
       Location.Z {type=MovieSceneScriptingDoubleChannel, default=1918.0084}
       Rotation.X {type=MovieSceneScriptingDoubleChannel, default=0}
       Rotation.Y {type=MovieSceneScriptingDoubleChannel, default=0}
       Rotation.Z {type=MovieSceneScriptingDoubleChannel, default=0}
       Scale.X {type=MovieSceneScriptingDoubleChannel, default=1}
       Scale.Y {type=MovieSceneScriptingDoubleChannel, default=1}
       Scale.Z {type=MovieSceneScriptingDoubleChannel, default=1}
      ]
     }]
   }
   {type=MovieSceneSkeletalAnimationTrack, sections [{type=MovieSceneSkeletalAnimationSection, end_frame=228, start_frame=0}]}
  ]
 }
]
display_rate {denominator=1, numerator=30}
tick_resolution {denominator=1, numerator=24000}
tracks []
