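# Niagara effect for the CampWrench emote: a layered, mostly blue shield mesh that pops in, swells, pulses and fades.
# Only Shield_Option2 is enabled; Shield_Old and Shield_Option1 are disabled earlier variants.
# A five-particle burst is partitioned by index into VisibilityTag layers for five mesh renderers, each with its own color and camera offset.
# Curves drive scale, wobble, alpha and material parameters, while system variables supply delay, placement, scale and time-of-day tint.
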NiagaraSystem /BRCosmetics/Animation/Game/MainPlayer/Emotes/CampWrench/FX/Systems/NS_Emote_CampWrench
dependencies {
 usedDataInterfaces [&NiagaraDataInterfaceCurve, &NiagaraDataInterfaceColorCurve]
 usedDynamicInputs [&ColorFromCurve, &FN_FrontEndFloatModifier]
 usedModules [
  &self:SystemSpawnScript.SystemScriptSource.SystemScriptGraph.NiagaraNodeAssignment_4.SetVariables_#1
  &self:SystemSpawnScript.SystemScriptSource.SystemScriptGraph.NiagaraNodeAssignment_0.SetVariables_#2
  &SystemState
  &self:SystemSpawnScript.SystemScriptSource.SystemScriptGraph.NiagaraNodeAssignment_5.SetVariables_#3
 ]
}
emitters {
 Shield_Old {
  bEnabled=false
  emitterUpdateScript [
   &EmitterState(
    Inactive Response=::NewEnumerator1
    Life Cycle Mode=::NewEnumerator1
    Loop Behavior=::NewEnumerator1
    Loop Delay=$System.GlobalDelay
    UseLoopDelay=true
   )
   &SpawnBurst_Instantaneous(Spawn Count=5, Use Spawn Probability=false, UseLoopCountLimit=false)
  ]
  particleSpawnScript [
   &PartitionParticles(ID to Partition:NiagaraInt32=&ReturnExecIndex(), Number Of Partitions=5)
   /SetVariables(Particles.VisibilityTag=$Particles.PartitionParticles.Partition) {isSetParameters=true, module=SetVariables_#4}
   &InitializeParticle(
    Color=&MergeHelper_ColorSwitch(
     0=<rgba 0.8551 0.9396 1 0.02>
     1=<rgba 0.3159 0.715 1 0.5>
     2=<rgba 0 2.5 5 0.1>
     3=<rgba 0.25 0.6875 1 0.15>
     4=<rgba 0.2 0.6667 1 0.07>
     Index=$Particles.PartitionParticles.Partition
     Num=::NewEnumerator3
    )
    Lifetime=5
    Mesh Scale=&Multiply_VectorByFloat(
     Vector=&Multiply_VectorByFloat(Vector=<-2 -4 -3.5>, Float=$System.GlobalScale)
     Float=&MergeHelper_FloatSwitch(0=1, 1=1.05, 2=1, 3=1, 4=1, Index=$Particles.PartitionParticles.Partition, Num=::NewEnumerator3)
    )
    Mesh Scale Mode=::NewEnumerator4
    Position Offset=$System.PositionOffset
    UsePositionOffset=true
    Write Material Random=true
   )
   &CameraOffset(Camera Offset Amount=&MergeHelper_FloatSwitch(0=0, 1=10, 2=0, 3=0, 4=-5, Index=$Particles.PartitionParticles.Partition, Num=::NewEnumerator3))
   &InitialMeshOrientation(Mesh Axis To Orient=<0 0 0>, Mesh Orientation Mode=::NewEnumerator2, Rotation=$System.InitialMeshRotation, Use Rotation Vector=true)
  ]
  particleUpdateScript [
   &ParticleState()
   &UpdateMeshOrientation(Rotation Rate=&Waveform("Waveform [1]"=::NewEnumerator0, "[1] Frequency"=5, "[1] Amplitude Min/Max"=<-0.05 0.05>))
   &ScaleMeshSize(Scale Factor=&VectorFromFloat(Value=&FloatFromCurve(FloatCurve=DataInterface() {propertyValues {Curve {
         defaultValue=1
         keys [
          {arriveTangent=12.6708, interpMode=RCIM_Cubic, leaveTangent=12.6708, tangentMode=RCTM_User, time=0, value=0}
          {arriveTangent=1.2843, leaveTangent=1.2843, time=0.03, value=1}
          {arriveTangent=0.8621, interpMode=RCIM_Cubic, leaveTangent=0.8621, time=0.3, value=1}
          {arriveTangent=-0.2322, leaveTangent=-0.2322, time=0.32, value=1.25}
          {arriveTangent=-0.2582, interpMode=RCIM_Cubic, leaveTangent=-0.2582, tangentMode=RCTM_User, time=0.34, value=1.25}
          {arriveTangent=0.1667, interpMode=RCIM_Cubic, leaveTangent=0.1667, time=0.7, value=1.35}
          {arriveTangent=0.8929, interpMode=RCIM_Cubic, leaveTangent=0.8929, time=0.94, value=1.35}
          <at 0.98 1.6 RCIM_Cubic>
         ]
        }}})))
   &ScaleMeshSize(Scale Factor=&VectorFromFloat(Value=&Waveform(
      "Waveform [1]"=::NewEnumerator0
      "[1] Frequency"=&FloatFromCurve(FloatCurve=DataInterface() {propertyValues {Curve {
          defaultValue=1
          keys [
           {arriveTangent=12.6708, interpMode=RCIM_Cubic, leaveTangent=12.6708, tangentMode=RCTM_User, time=0, value=0}
           {arriveTangent=1.2843, leaveTangent=1.2843, time=0.03, value=5}
           {arriveTangent=10.3448, interpMode=RCIM_Cubic, leaveTangent=10.3448, time=0.3, value=5}
           {arriveTangent=-0.2322, leaveTangent=-0.2322, time=0.32, value=8}
           {arriveTangent=-0.2582, interpMode=RCIM_Cubic, leaveTangent=-0.2582, tangentMode=RCTM_User, time=0.34, value=8}
           {interpMode=RCIM_Cubic, leaveTangent=0.5469, time=0.7, value=8.5}
          ]
         }}})
      "[1] Amplitude Min/Max"=<0.99 1.01>
     ))) {module=ScaleMeshSize001}
   &DynamicMaterialParameters(
    Index 0 Param 1=1.3646
    Index 0 Param 3=&FloatFromCurve(FloatCurve=DataInterface() {propertyValues {Curve {keys [
         {arriveTangent=0.2359, interpMode=RCIM_Cubic, leaveTangent=0.2359, tangentMode=RCTM_User, time=0, value=1}
         {arriveTangent=-0.0509, tangentMode=RCTM_User, time=0.2, value=1.05}
         {arriveTangent=4.515, interpMode=RCIM_Cubic, leaveTangent=4.515, tangentMode=RCTM_User, time=0.3, value=1.05}
         {interpMode=RCIM_Cubic, leaveTangent=2.75, time=0.33, value=1.1}
        ]}}})
    Module.Index0Float0Write=true
    Module.Index0Float1Write=true
    Module.Index0Float2Write=true
    Module.Index0Float3Write=true
   )
   /SetVariables(
    Particles.DynPar_ChromaRipples:Vector4f=&MakeVector4(
     X=0.5
     Y=&FloatFromCurve(FloatCurve=DataInterface() {propertyValues {Curve {keys [<at 0.94 1 RCIM_Cubic>,<at 1 0>]}}})
     Z=0
     W=0
    )
    Particles.DynPar_ChromaWave:Vector4f=&MakeVector4(
     X=&RandomRangeFloat(Minimum=0, Maximum=1)
     Y=&FloatFromCurve(
      FloatCurve=DataInterface() {propertyValues {Curve {defaultValue=0, keys [{tangentMode=RCTM_User, time=0, value=1}, {tangentMode=RCTM_User, time=1, value=0}]}}}
      Scale Curve=2
     )
     Z=0
     W=0
    )
    Particles.DynPar_SoftWave:Vector4f=&MakeVector4(X=3, Y=1.5, Z=0, W=0)
   ) {isSetParameters=true, module=SetVariables_#5}
   &ScaleColor(
    Scale Alpha=&Multiply_Float(
     A=&MakeFloatFromLinearColor(LinearColor=$System.TODColorMultiplier, Channel:ENiagaraFloatFromLinearColorOptions=::NewEnumerator3)
     B=&MergeHelper_FloatSwitch(
      0=&FloatFromCurve(FloatCurve=DataInterface() {propertyValues {Curve {
          defaultValue=1
          keys [
           {arriveTangent=14.3464, interpMode=RCIM_Cubic, leaveTangent=14.3464, tangentMode=RCTM_User, time=0, value=0}
           {arriveTangent=1.2843, leaveTangent=1.2843, time=0.08, value=0.6}
           {arriveTangent=1.6667, interpMode=RCIM_Cubic, leaveTangent=1.6667, time=0.3, value=0.6}
           {arriveTangent=-0.2322, leaveTangent=-0.2322, time=0.32, value=1}
           {arriveTangent=-0.2582, interpMode=RCIM_Cubic, leaveTangent=-0.2582, tangentMode=RCTM_User, time=0.34, value=1}
           {arriveTangent=0.1667, interpMode=RCIM_Cubic, leaveTangent=0.1667, time=0.7, value=1.1}
           {arriveTangent=0.0509, interpMode=RCIM_Cubic, leaveTangent=-6.389, tangentMode=RCTM_Break, time=0.94, value=1.1}
           <at 0.98 0 RCIM_Cubic>
          ]
         }}})
      1=&FloatFromCurve(FloatCurve=DataInterface() {propertyValues {Curve {
          defaultValue=1
          keys [
           {arriveTangent=14.3464, interpMode=RCIM_Cubic, leaveTangent=14.3464, tangentMode=RCTM_User, time=0, value=0}
           {arriveTangent=1.2843, leaveTangent=1.2843, time=0.08, value=0.6}
           {arriveTangent=1.6667, interpMode=RCIM_Cubic, leaveTangent=1.6667, time=0.3, value=0.6}
           {arriveTangent=-0.2322, leaveTangent=-0.2322, time=0.32, value=1}
           {arriveTangent=-0.2582, interpMode=RCIM_Cubic, leaveTangent=-0.2582, tangentMode=RCTM_User, time=0.34, value=1}
           {arriveTangent=0.1667, interpMode=RCIM_Cubic, leaveTangent=0.1667, time=0.7, value=1.1}
           {arriveTangent=0.0509, interpMode=RCIM_Cubic, leaveTangent=-6.389, tangentMode=RCTM_Break, time=0.94, value=1.1}
           <at 0.98 0 RCIM_Cubic>
          ]
         }}})
      2=&FloatFromCurve(FloatCurve=DataInterface() {propertyValues {Curve {
          defaultValue=1
          keys [
           {arriveTangent=14.3464, interpMode=RCIM_Cubic, leaveTangent=14.3464, tangentMode=RCTM_User, time=0, value=0}
           {arriveTangent=1.2843, leaveTangent=1.2843, time=0.08, value=0.6}
           {arriveTangent=1.6667, interpMode=RCIM_Cubic, leaveTangent=1.6667, time=0.3, value=0.6}
           {arriveTangent=-0.2322, leaveTangent=-0.2322, time=0.32, value=1}
           {arriveTangent=-0.2582, interpMode=RCIM_Cubic, leaveTangent=-0.2582, tangentMode=RCTM_User, time=0.34, value=1}
           {arriveTangent=0.1613, interpMode=RCIM_Cubic, leaveTangent=0.1613, time=0.7, value=1.1}
           {arriveTangent=0.0509, interpMode=RCIM_Cubic, leaveTangent=-6.389, tangentMode=RCTM_Break, time=0.96, value=1.1}
           <at 1 0.4 RCIM_Cubic>
          ]
         }}})
      3=&FloatFromCurve(FloatCurve=DataInterface() {propertyValues {Curve {
          defaultValue=1
          keys [
           {arriveTangent=14.3464, interpMode=RCIM_Cubic, leaveTangent=14.3464, tangentMode=RCTM_User, time=0, value=0}
           {arriveTangent=1.2843, leaveTangent=1.2843, time=0.08, value=0.6}
           {arriveTangent=1.6667, interpMode=RCIM_Cubic, leaveTangent=1.6667, time=0.3, value=0.6}
           {arriveTangent=-0.2322, leaveTangent=-0.2322, time=0.32, value=1}
           {arriveTangent=-0.2582, interpMode=RCIM_Cubic, leaveTangent=-0.2582, tangentMode=RCTM_User, time=0.34, value=1}
           {arriveTangent=0.1667, interpMode=RCIM_Cubic, leaveTangent=0.1667, time=0.7, value=1.1}
           {arriveTangent=0.0509, interpMode=RCIM_Cubic, leaveTangent=-6.389, tangentMode=RCTM_Break, time=0.94, value=1.1}
           <at 0.98 0 RCIM_Cubic>
          ]
         }}})
      4=&FloatFromCurve(FloatCurve=DataInterface() {propertyValues {Curve {
          defaultValue=1
          keys [
           {arriveTangent=14.3464, interpMode=RCIM_Cubic, leaveTangent=14.3464, tangentMode=RCTM_User, time=0, value=0}
           {arriveTangent=1.2843, leaveTangent=1.2843, time=0.08, value=0.6}
           {arriveTangent=1.6667, interpMode=RCIM_Cubic, leaveTangent=1.6667, time=0.3, value=0.6}
           {arriveTangent=-0.2322, leaveTangent=-0.2322, time=0.32, value=1}
           {arriveTangent=-0.2582, interpMode=RCIM_Cubic, leaveTangent=-0.2582, tangentMode=RCTM_User, time=0.34, value=1}
           {arriveTangent=0.1667, interpMode=RCIM_Cubic, leaveTangent=0.1667, time=0.7, value=1.1}
           {arriveTangent=0.0509, interpMode=RCIM_Cubic, leaveTangent=-6.389, tangentMode=RCTM_Break, time=0.94, value=1.1}
           <at 0.98 0 RCIM_Cubic>
          ]
         }}})
      Index=$Particles.PartitionParticles.Partition
      Num=::NewEnumerator3
     )
    )
    Scale RGB=&MakeVectorFromLinearColorRGB(Color=$System.TODColorMultiplier)
    ScaleA=true
    ScaleRGB=true
   ) {module=ScaleColor001}
  ]
  rendererClasses [&NiagaraMeshRendererProperties]
  renderers [
   {rendererClass=&NiagaraMeshRendererProperties, rendererIndex=0}
   {rendererClass=&NiagaraMeshRendererProperties, rendererIndex=1}
   {rendererClass=&NiagaraMeshRendererProperties, rendererIndex=2}
   {rendererClass=&NiagaraMeshRendererProperties, rendererIndex=3}
   {rendererClass=&NiagaraMeshRendererProperties, rendererIndex=4}
  ]
 }
 Shield_Option1 {
  bEnabled=false
  emitterUpdateScript [
   &EmitterState(
    Inactive Response=::NewEnumerator1
    Life Cycle Mode=::NewEnumerator1
    Loop Behavior=::NewEnumerator1
    Loop Delay=$System.GlobalDelay
    UseLoopDelay=true
   )
   &SpawnBurst_Instantaneous(Spawn Count=5, Use Spawn Probability=false, UseLoopCountLimit=false)
   /SetVariables(Emitter.GlowPower:NiagaraFloat=&FloatFromCurve(
     FloatCurve=DataInterface() {propertyValues {Curve {
        defaultValue=1
        keys [
         {arriveTangent=14.3464, time=0, value=-3}
         {arriveTangent=6.25, leaveTangent=6.25, time=0.3, value=-3}
         {arriveTangent=-0.2322, leaveTangent=-0.2322, time=0.32, value=-1}
         {arriveTangent=-0.2582, interpMode=RCIM_Cubic, leaveTangent=-0.2582, tangentMode=RCTM_User, time=0.34, value=-1}
         {interpMode=RCIM_Cubic, leaveTangent=1.5625, time=0.7, value=-1.5}
        ]
       }}}
     CurveIndex=$Emitter.NormalizedLoopAge
    )) {isSetParameters=true, module=SetVariables_#6}
  ]
  particleSpawnScript [
   &PartitionParticles(ID to Partition:NiagaraInt32=&ReturnExecIndex(), Number Of Partitions=5)
   /SetVariables(Particles.VisibilityTag=$Particles.PartitionParticles.Partition) {isSetParameters=true, module=SetVariables_#4}
   &InitializeParticle(
    Color=&MergeHelper_ColorSwitch(
     0=<rgba 0.8551 0.9396 1 0.15>
     1=<rgba 0.3159 0.715 1 0.5>
     2=<rgba 0 2.5 5 0.1>
     3=<rgba 0.25 0.6875 1 0.15>
     4=<rgba 0.2 0.6667 1 0.03>
     Index=$Particles.PartitionParticles.Partition
     Num=::NewEnumerator3
    )
    Lifetime=5
    Mesh Scale=&Multiply_VectorByFloat(Vector=&Multiply_VectorByFloat(Vector=<-2 -4 -3.5>, Float=$System.GlobalScale), Float=1)
    Mesh Scale Mode=::NewEnumerator4
    Position Offset=$System.PositionOffset
    UsePositionOffset=true
    Write Material Random=true
   )
   &CameraOffset(Camera Offset Amount=&MergeHelper_FloatSwitch(0=0, 1=10, 2=0, 3=0, 4=-5, Index=$Particles.PartitionParticles.Partition, Num=::NewEnumerator3))
   &InitialMeshOrientation(Mesh Axis To Orient=<0 0 0>, Mesh Orientation Mode=::NewEnumerator2, Rotation=$System.InitialMeshRotation, Use Rotation Vector=true)
  ]
  particleUpdateScript [
   &ParticleState()
   &UpdateMeshOrientation(Rotation Rate=&Waveform("Waveform [1]"=::NewEnumerator0, "[1] Frequency"=5, "[1] Amplitude Min/Max"=<-0.05 0.05>))
   &ScaleMeshSize(Scale Factor=&VectorFromFloat(Value=&FloatFromCurve(FloatCurve=DataInterface() {propertyValues {Curve {
         defaultValue=1
         keys [
          {arriveTangent=12.6708, interpMode=RCIM_Cubic, leaveTangent=12.6708, tangentMode=RCTM_User, time=0, value=0}
          {arriveTangent=1.2843, leaveTangent=1.2843, time=0.03, value=1}
          {arriveTangent=0.8621, interpMode=RCIM_Cubic, leaveTangent=0.8621, time=0.3, value=1}
          {arriveTangent=-0.2322, leaveTangent=-0.2322, time=0.32, value=1.25}
          {arriveTangent=-0.2582, interpMode=RCIM_Cubic, leaveTangent=-0.2582, tangentMode=RCTM_User, time=0.34, value=1.25}
          {arriveTangent=0.1667, interpMode=RCIM_Cubic, leaveTangent=0.1667, time=0.7, value=1.35}
          {arriveTangent=0.8929, interpMode=RCIM_Cubic, leaveTangent=0.8929, time=0.94, value=1.35}
          <at 0.98 1.6 RCIM_Cubic>
         ]
        }}})))
   &ScaleMeshSize(Scale Factor=&VectorFromFloat(Value=&Waveform(
      "Waveform [1]"=::NewEnumerator0
      "[1] Frequency"=&FloatFromCurve(FloatCurve=DataInterface() {propertyValues {Curve {
          defaultValue=1
          keys [
           {arriveTangent=12.6708, interpMode=RCIM_Cubic, leaveTangent=12.6708, tangentMode=RCTM_User, time=0, value=0}
           {arriveTangent=1.2843, leaveTangent=1.2843, time=0.03, value=5}
           {arriveTangent=10.3448, interpMode=RCIM_Cubic, leaveTangent=10.3448, time=0.3, value=5}
           {arriveTangent=-0.2322, leaveTangent=-0.2322, time=0.32, value=8}
           {arriveTangent=-0.2582, interpMode=RCIM_Cubic, leaveTangent=-0.2582, tangentMode=RCTM_User, time=0.34, value=8}
           {interpMode=RCIM_Cubic, leaveTangent=0.5469, time=0.7, value=8.5}
          ]
         }}})
      "[1] Amplitude Min/Max"=<0.99 1.01>
     ))) {module=ScaleMeshSize001}
   &DynamicMaterialParameters(
    Index 0 Param 1=1.3646
    Index 0 Param 3=&FloatFromCurve(FloatCurve=DataInterface() {propertyValues {Curve {keys [
         {arriveTangent=0.2359, interpMode=RCIM_Cubic, leaveTangent=0.2359, tangentMode=RCTM_User, time=0, value=1}
         {arriveTangent=-0.0509, tangentMode=RCTM_User, time=0.2, value=1.05}
         {arriveTangent=4.515, interpMode=RCIM_Cubic, leaveTangent=4.515, tangentMode=RCTM_User, time=0.3, value=1.05}
         {interpMode=RCIM_Cubic, leaveTangent=2.75, time=0.33, value=1.1}
        ]}}})
    Module.Index0Float0Write=true
    Module.Index0Float1Write=true
    Module.Index0Float2Write=true
    Module.Index0Float3Write=true
   )
   /SetVariables(
    Particles.DynPar_ChromaRipples:Vector4f=&MakeVector4(
     X=0.5
     Y=&FloatFromCurve(FloatCurve=DataInterface() {propertyValues {Curve {keys [<at 0.94 1 RCIM_Cubic>,<at 1 0>]}}})
     Z=0
     W=0
    )
    Particles.DynPar_ChromaWave:Vector4f=&MakeVector4(
     X=&RandomRangeFloat(Minimum=0, Maximum=1)
     Y=&FloatFromCurve(
      FloatCurve=DataInterface() {propertyValues {Curve {defaultValue=0, keys [{tangentMode=RCTM_User, time=0, value=1}, {tangentMode=RCTM_User, time=1, value=0}]}}}
      Scale Curve=2
     )
     Z=0
     W=0
    )
    Particles.DynPar_SoftWave:Vector4f=&MakeVector4(X=3, Y=1.5, Z=0, W=0)
   ) {isSetParameters=true, module=SetVariables_#5}
   &ScaleColor(
    Scale Alpha=&Multiply_Float(
     A=&MakeFloatFromLinearColor(LinearColor=$System.TODColorMultiplier, Channel:ENiagaraFloatFromLinearColorOptions=::NewEnumerator3)
     B=&MergeHelper_FloatSwitch(
      0=&FloatFromCurve(FloatCurve=DataInterface() {propertyValues {Curve {
          defaultValue=1
          keys [
           {arriveTangent=14.3464, interpMode=RCIM_Cubic, leaveTangent=14.3464, tangentMode=RCTM_User, time=0, value=0}
           {arriveTangent=1.2843, leaveTangent=1.2843, time=0.08, value=0.6}
           {arriveTangent=1.6667, interpMode=RCIM_Cubic, leaveTangent=1.6667, time=0.3, value=0.6}
           {arriveTangent=-0.2322, leaveTangent=-0.2322, time=0.32, value=1}
           {arriveTangent=-0.2582, interpMode=RCIM_Cubic, leaveTangent=-0.2582, tangentMode=RCTM_User, time=0.34, value=1}
           {arriveTangent=0.1667, interpMode=RCIM_Cubic, leaveTangent=0.1667, time=0.7, value=1.1}
           {arriveTangent=0.0509, interpMode=RCIM_Cubic, leaveTangent=-6.389, tangentMode=RCTM_Break, time=0.94, value=1.1}
           <at 0.98 0 RCIM_Cubic>
          ]
         }}})
      1=&FloatFromCurve(FloatCurve=DataInterface() {propertyValues {Curve {
          defaultValue=1
          keys [
           {arriveTangent=14.3464, interpMode=RCIM_Cubic, leaveTangent=14.3464, tangentMode=RCTM_User, time=0, value=0}
           {arriveTangent=1.2843, leaveTangent=1.2843, time=0.08, value=0.6}
           {arriveTangent=1.6667, interpMode=RCIM_Cubic, leaveTangent=1.6667, time=0.3, value=0.6}
           {arriveTangent=-0.2322, leaveTangent=-0.2322, time=0.32, value=1}
           {arriveTangent=-0.2582, interpMode=RCIM_Cubic, leaveTangent=-0.2582, tangentMode=RCTM_User, time=0.34, value=1}
           {arriveTangent=0.1667, interpMode=RCIM_Cubic, leaveTangent=0.1667, time=0.7, value=1.1}
           {arriveTangent=0.0509, interpMode=RCIM_Cubic, leaveTangent=-6.389, tangentMode=RCTM_Break, time=0.94, value=1.1}
           <at 0.98 0 RCIM_Cubic>
          ]
         }}})
      2=&FloatFromCurve(FloatCurve=DataInterface() {propertyValues {Curve {
          defaultValue=1
          keys [
           {arriveTangent=14.3464, interpMode=RCIM_Cubic, leaveTangent=14.3464, tangentMode=RCTM_User, time=0, value=0}
           {arriveTangent=1.2843, leaveTangent=1.2843, time=0.08, value=0.6}
           {arriveTangent=1.6667, interpMode=RCIM_Cubic, leaveTangent=1.6667, time=0.3, value=0.6}
           {arriveTangent=-0.2322, leaveTangent=-0.2322, time=0.32, value=1}
           {arriveTangent=-0.2582, interpMode=RCIM_Cubic, leaveTangent=-0.2582, tangentMode=RCTM_User, time=0.34, value=1}
           {arriveTangent=0.1613, interpMode=RCIM_Cubic, leaveTangent=0.1613, time=0.7, value=1.1}
           {arriveTangent=0.0509, interpMode=RCIM_Cubic, leaveTangent=-6.389, tangentMode=RCTM_Break, time=0.96, value=1.1}
           <at 1 0.4 RCIM_Cubic>
          ]
         }}})
      3=&FloatFromCurve(FloatCurve=DataInterface() {propertyValues {Curve {
          defaultValue=1
          keys [
           {arriveTangent=14.3464, interpMode=RCIM_Cubic, leaveTangent=14.3464, tangentMode=RCTM_User, time=0, value=0}
           {arriveTangent=1.2843, leaveTangent=1.2843, time=0.08, value=0.6}
           {arriveTangent=1.6667, interpMode=RCIM_Cubic, leaveTangent=1.6667, time=0.3, value=0.6}
           {arriveTangent=-0.2322, leaveTangent=-0.2322, time=0.32, value=1}
           {arriveTangent=-0.2582, interpMode=RCIM_Cubic, leaveTangent=-0.2582, tangentMode=RCTM_User, time=0.34, value=1}
           {arriveTangent=0.1667, interpMode=RCIM_Cubic, leaveTangent=0.1667, time=0.7, value=1.1}
           {arriveTangent=0.0509, interpMode=RCIM_Cubic, leaveTangent=-6.389, tangentMode=RCTM_Break, time=0.94, value=1.1}
           <at 0.98 0 RCIM_Cubic>
          ]
         }}})
      4=&FloatFromCurve(FloatCurve=DataInterface() {propertyValues {Curve {
          defaultValue=1
          keys [
           {arriveTangent=14.3464, interpMode=RCIM_Cubic, leaveTangent=14.3464, tangentMode=RCTM_User, time=0, value=0}
           {arriveTangent=1.2843, leaveTangent=1.2843, time=0.08, value=0.6}
           {arriveTangent=1.6667, interpMode=RCIM_Cubic, leaveTangent=1.6667, time=0.3, value=0.6}
           {arriveTangent=-0.2322, leaveTangent=-0.2322, time=0.32, value=1}
           {arriveTangent=-0.2582, interpMode=RCIM_Cubic, leaveTangent=-0.2582, tangentMode=RCTM_User, time=0.34, value=1}
           {arriveTangent=0.1667, interpMode=RCIM_Cubic, leaveTangent=0.1667, time=0.7, value=1.1}
           {arriveTangent=0.0509, interpMode=RCIM_Cubic, leaveTangent=-6.389, tangentMode=RCTM_Break, time=0.94, value=1.1}
           <at 0.98 0 RCIM_Cubic>
          ]
         }}})
      Index=$Particles.PartitionParticles.Partition
      Num=::NewEnumerator3
     )
    )
    Scale RGB=&MakeVectorFromLinearColorRGB(Color=$System.TODColorMultiplier)
    ScaleA=true
    ScaleRGB=true
   ) {module=ScaleColor001}
   &ScaleColor(ScaleA=true, ScaleRGB=true) {enabled=false}
  ]
  rendererClasses [&NiagaraMeshRendererProperties]
  renderers [
   {rendererClass=&NiagaraMeshRendererProperties, rendererIndex=0}
   {rendererClass=&NiagaraMeshRendererProperties, rendererIndex=1}
   {rendererClass=&NiagaraMeshRendererProperties, rendererIndex=2}
   {rendererClass=&NiagaraMeshRendererProperties, rendererIndex=3}
   {rendererClass=&NiagaraMeshRendererProperties, rendererIndex=4}
  ]
 }
 Shield_Option2 {
  emitterUpdateScript [
   &EmitterState(
    Inactive Response=::NewEnumerator1
    Life Cycle Mode=::NewEnumerator1
    Loop Behavior=::NewEnumerator1
    Loop Delay=$System.GlobalDelay
    UseLoopDelay=true
   )
   &SpawnBurst_Instantaneous(Spawn Count=5, Use Spawn Probability=false, UseLoopCountLimit=false)
   /SetVariables(Emitter.GlowPower:NiagaraFloat=&FloatFromCurve(
     FloatCurve=DataInterface() {propertyValues {Curve {
        defaultValue=1
        keys [
         {arriveTangent=14.3464, time=0, value=-3}
         {arriveTangent=6.25, leaveTangent=6.25, time=0.3, value=-3}
         {arriveTangent=-0.2322, leaveTangent=-0.2322, time=0.32, value=-1}
         {arriveTangent=-0.2582, interpMode=RCIM_Cubic, leaveTangent=-0.2582, tangentMode=RCTM_User, time=0.34, value=-1}
         {interpMode=RCIM_Cubic, leaveTangent=1.5625, time=0.7, value=-1.5}
        ]
       }}}
     CurveIndex=$Emitter.NormalizedLoopAge
    )) {isSetParameters=true, module=SetVariables_#6}
  ]
  particleSpawnScript [
   &PartitionParticles(ID to Partition:NiagaraInt32=&ReturnExecIndex(), Number Of Partitions=5)
   /SetVariables(Particles.VisibilityTag=$Particles.PartitionParticles.Partition) {isSetParameters=true, module=SetVariables_#4}
   &InitializeParticle(
    Color=&MergeHelper_ColorSwitch(
     0=<rgba 0.8551 0.9396 1 0.15>
     1=<rgba 0.3159 0.715 1 0.5>
     2=<rgba 0 2.5 5 0.1>
     3=<rgba 0.25 0.6875 1 0.15>
     4=<rgba 0.2 0.6667 1 0.03>
     Index=$Particles.PartitionParticles.Partition
     Num=::NewEnumerator3
    )
    Lifetime=5
    Mesh Scale=&Multiply_VectorByFloat(Vector=&Multiply_VectorByFloat(Vector=<-2 -4 -3.5>, Float=$System.GlobalScale), Float=1)
    Mesh Scale Mode=::NewEnumerator4
    Position Offset=$System.PositionOffset
    UsePositionOffset=true
    Write Material Random=true
   )
   &CameraOffset(Camera Offset Amount=&MergeHelper_FloatSwitch(0=0, 1=10, 2=0, 3=0, 4=-5, Index=$Particles.PartitionParticles.Partition, Num=::NewEnumerator3))
   &InitialMeshOrientation(Mesh Axis To Orient=<0 0 0>, Mesh Orientation Mode=::NewEnumerator2, Rotation=$System.InitialMeshRotation, Use Rotation Vector=true)
  ]
  particleUpdateScript [
   &ParticleState()
   &UpdateMeshOrientation(Rotation Rate=&Waveform("Waveform [1]"=::NewEnumerator0, "[1] Frequency"=5, "[1] Amplitude Min/Max"=<-0.05 0.05>))
   &ScaleMeshSize(Scale Factor=&VectorFromFloat(Value=&FloatFromCurve(FloatCurve=DataInterface() {propertyValues {Curve {
         defaultValue=1
         keys [
          {arriveTangent=12.6708, interpMode=RCIM_Cubic, leaveTangent=12.6708, tangentMode=RCTM_User, time=0, value=0}
          {arriveTangent=1.2843, leaveTangent=1.2843, time=0.03, value=1}
          {arriveTangent=0.8621, interpMode=RCIM_Cubic, leaveTangent=0.8621, time=0.3, value=1}
          {arriveTangent=-0.2322, leaveTangent=-0.2322, time=0.32, value=1.25}
          {arriveTangent=-0.2582, interpMode=RCIM_Cubic, leaveTangent=-0.2582, tangentMode=RCTM_User, time=0.34, value=1.25}
          {arriveTangent=0.1667, interpMode=RCIM_Cubic, leaveTangent=0.1667, time=0.7, value=1.35}
          {arriveTangent=0.8929, interpMode=RCIM_Cubic, leaveTangent=0.8929, time=0.94, value=1.35}
          <at 0.98 1.6 RCIM_Cubic>
         ]
        }}})))
   &ScaleMeshSize(Scale Factor=&VectorFromFloat(Value=&Waveform(
      "Waveform [1]"=::NewEnumerator0
      "[1] Frequency"=&FloatFromCurve(FloatCurve=DataInterface() {propertyValues {Curve {
          defaultValue=1
          keys [
           {arriveTangent=12.6708, interpMode=RCIM_Cubic, leaveTangent=12.6708, tangentMode=RCTM_User, time=0, value=0}
           {arriveTangent=1.2843, leaveTangent=1.2843, time=0.03, value=5}
           {arriveTangent=10.3448, interpMode=RCIM_Cubic, leaveTangent=10.3448, time=0.3, value=5}
           {arriveTangent=-0.2322, leaveTangent=-0.2322, time=0.32, value=8}
           {arriveTangent=-0.2582, interpMode=RCIM_Cubic, leaveTangent=-0.2582, tangentMode=RCTM_User, time=0.34, value=8}
           {interpMode=RCIM_Cubic, leaveTangent=0.5469, time=0.7, value=8.5}
          ]
         }}})
      "[1] Amplitude Min/Max"=<0.99 1.01>
     ))) {module=ScaleMeshSize001}
   &DynamicMaterialParameters(
    Index 0 Param 1=1.3646
    Index 0 Param 3=&FloatFromCurve(FloatCurve=DataInterface() {propertyValues {Curve {keys [
         {arriveTangent=0.2359, interpMode=RCIM_Cubic, leaveTangent=0.2359, tangentMode=RCTM_User, time=0, value=1}
         {arriveTangent=-0.0509, tangentMode=RCTM_User, time=0.2, value=1.05}
         {arriveTangent=4.515, interpMode=RCIM_Cubic, leaveTangent=4.515, tangentMode=RCTM_User, time=0.3, value=1.05}
         {interpMode=RCIM_Cubic, leaveTangent=2.75, time=0.33, value=1.1}
        ]}}})
    Module.Index0Float0Write=true
    Module.Index0Float1Write=true
    Module.Index0Float2Write=true
    Module.Index0Float3Write=true
   )
   /SetVariables(
    Particles.DynPar_ChromaRipples:Vector4f=&MakeVector4(
     X=0.5
     Y=&FloatFromCurve(FloatCurve=DataInterface() {propertyValues {Curve {keys [<at 0.94 1 RCIM_Cubic>,<at 1 0>]}}})
     Z=0
     W=0
    )
    Particles.DynPar_ChromaWave:Vector4f=&MakeVector4(
     X=&RandomRangeFloat(Minimum=0, Maximum=1)
     Y=&FloatFromCurve(
      FloatCurve=DataInterface() {propertyValues {Curve {defaultValue=0, keys [{tangentMode=RCTM_User, time=0, value=1}, {tangentMode=RCTM_User, time=1, value=0}]}}}
      Scale Curve=2
     )
     Z=0
     W=0
    )
    Particles.DynPar_SoftWave:Vector4f=&MakeVector4(X=3, Y=1.5, Z=0, W=0)
   ) {isSetParameters=true, module=SetVariables_#5}
   &ScaleColor(
    Scale Alpha=&Multiply_Float(
     A=&MakeFloatFromLinearColor(LinearColor=$System.TODColorMultiplier, Channel:ENiagaraFloatFromLinearColorOptions=::NewEnumerator3)
     B=&MergeHelper_FloatSwitch(
      0=&FloatFromCurve(FloatCurve=DataInterface() {propertyValues {Curve {
          defaultValue=1
          keys [
           {arriveTangent=14.3464, interpMode=RCIM_Cubic, leaveTangent=14.3464, tangentMode=RCTM_User, time=0, value=0}
           {arriveTangent=1.2843, leaveTangent=1.2843, time=0.08, value=0.6}
           {arriveTangent=1.6667, interpMode=RCIM_Cubic, leaveTangent=1.6667, time=0.3, value=0.6}
           {arriveTangent=-0.2322, leaveTangent=-0.2322, time=0.32, value=1}
           {arriveTangent=-0.2582, interpMode=RCIM_Cubic, leaveTangent=-0.2582, tangentMode=RCTM_User, time=0.34, value=1}
           {arriveTangent=0.1667, interpMode=RCIM_Cubic, leaveTangent=0.1667, time=0.7, value=1.1}
           {arriveTangent=0.0509, interpMode=RCIM_Cubic, leaveTangent=-6.389, tangentMode=RCTM_Break, time=0.94, value=1.1}
           <at 0.98 0 RCIM_Cubic>
          ]
         }}})
      1=&FloatFromCurve(FloatCurve=DataInterface() {propertyValues {Curve {
          defaultValue=1
          keys [
           {arriveTangent=14.3464, interpMode=RCIM_Cubic, leaveTangent=14.3464, tangentMode=RCTM_User, time=0, value=0}
           {arriveTangent=1.2843, leaveTangent=1.2843, time=0.08, value=0.6}
           {arriveTangent=1.6667, interpMode=RCIM_Cubic, leaveTangent=1.6667, time=0.3, value=0.6}
           {arriveTangent=-0.2322, leaveTangent=-0.2322, time=0.32, value=1}
           {arriveTangent=-0.2582, interpMode=RCIM_Cubic, leaveTangent=-0.2582, tangentMode=RCTM_User, time=0.34, value=1}
           {arriveTangent=0.1667, interpMode=RCIM_Cubic, leaveTangent=0.1667, time=0.7, value=1.1}
           {arriveTangent=0.0509, interpMode=RCIM_Cubic, leaveTangent=-6.389, tangentMode=RCTM_Break, time=0.94, value=1.1}
           <at 0.98 0 RCIM_Cubic>
          ]
         }}})
      2=&FloatFromCurve(FloatCurve=DataInterface() {propertyValues {Curve {
          defaultValue=1
          keys [
           {arriveTangent=14.3464, interpMode=RCIM_Cubic, leaveTangent=14.3464, tangentMode=RCTM_User, time=0, value=0}
           {arriveTangent=1.2843, leaveTangent=1.2843, time=0.08, value=0.6}
           {arriveTangent=1.6667, interpMode=RCIM_Cubic, leaveTangent=1.6667, time=0.3, value=0.6}
           {arriveTangent=-0.2322, leaveTangent=-0.2322, time=0.32, value=1}
           {arriveTangent=-0.2582, interpMode=RCIM_Cubic, leaveTangent=-0.2582, tangentMode=RCTM_User, time=0.34, value=1}
           {arriveTangent=0.1613, interpMode=RCIM_Cubic, leaveTangent=0.1613, time=0.7, value=1.1}
           {arriveTangent=0.0509, interpMode=RCIM_Cubic, leaveTangent=-6.389, tangentMode=RCTM_Break, time=0.96, value=1.1}
           <at 1 0.4 RCIM_Cubic>
          ]
         }}})
      3=&FloatFromCurve(FloatCurve=DataInterface() {propertyValues {Curve {
          defaultValue=1
          keys [
           {arriveTangent=14.3464, interpMode=RCIM_Cubic, leaveTangent=14.3464, tangentMode=RCTM_User, time=0, value=0}
           {arriveTangent=1.2843, leaveTangent=1.2843, time=0.08, value=0.6}
           {arriveTangent=1.6667, interpMode=RCIM_Cubic, leaveTangent=1.6667, time=0.3, value=0.6}
           {arriveTangent=-0.2322, leaveTangent=-0.2322, time=0.32, value=1}
           {arriveTangent=-0.2582, interpMode=RCIM_Cubic, leaveTangent=-0.2582, tangentMode=RCTM_User, time=0.34, value=1}
           {arriveTangent=0.1667, interpMode=RCIM_Cubic, leaveTangent=0.1667, time=0.7, value=1.1}
           {arriveTangent=0.0509, interpMode=RCIM_Cubic, leaveTangent=-6.389, tangentMode=RCTM_Break, time=0.94, value=1.1}
           <at 0.98 0 RCIM_Cubic>
          ]
         }}})
      4=&FloatFromCurve(FloatCurve=DataInterface() {propertyValues {Curve {
          defaultValue=1
          keys [
           {arriveTangent=14.3464, interpMode=RCIM_Cubic, leaveTangent=14.3464, tangentMode=RCTM_User, time=0, value=0}
           {arriveTangent=1.2843, leaveTangent=1.2843, time=0.08, value=0.6}
           {arriveTangent=1.6667, interpMode=RCIM_Cubic, leaveTangent=1.6667, time=0.3, value=0.6}
           {arriveTangent=-0.2322, leaveTangent=-0.2322, time=0.32, value=1}
           {arriveTangent=-0.2582, interpMode=RCIM_Cubic, leaveTangent=-0.2582, tangentMode=RCTM_User, time=0.34, value=1}
           {arriveTangent=0.1667, interpMode=RCIM_Cubic, leaveTangent=0.1667, time=0.7, value=1.1}
           {arriveTangent=0.0509, interpMode=RCIM_Cubic, leaveTangent=-6.389, tangentMode=RCTM_Break, time=0.94, value=1.1}
           <at 0.98 0 RCIM_Cubic>
          ]
         }}})
      Index=$Particles.PartitionParticles.Partition
      Num=::NewEnumerator3
     )
    )
    Scale RGB=&MakeVectorFromLinearColorRGB(Color=$System.TODColorMultiplier)
    ScaleA=true
    ScaleRGB=true
   ) {module=ScaleColor001}
   &ScaleColor(ScaleA=true, ScaleRGB=true) {enabled=false}
   /SetVariables(Particles.FillColor:LinearColor=&Multiply_LinearColorByFloat(Linear Color=$Particles.Color, Float=0.15)) {isSetParameters=true}
  ]
  rendererClasses [&NiagaraMeshRendererProperties]
  renderers [
   {rendererClass=&NiagaraMeshRendererProperties, rendererIndex=0}
   {rendererClass=&NiagaraMeshRendererProperties, rendererIndex=1}
   {rendererClass=&NiagaraMeshRendererProperties, rendererIndex=2}
   {rendererClass=&NiagaraMeshRendererProperties, rendererIndex=3}
   {rendererClass=&NiagaraMeshRendererProperties, rendererIndex=4}
  ]
 }
}
summary {
 systemName=NS_Emote_CampWrench
 emitters [
  {bEnabled=false, emitterName=Shield_Old, rendererClasses [&NiagaraMeshRendererProperties]}
  {bEnabled=false, emitterName=Shield_Option1, rendererClasses [&NiagaraMeshRendererProperties]}
  {emitterName=Shield_Option2, rendererClasses [&NiagaraMeshRendererProperties]}
 ]
 userVariables [
  User.bIsFrontEnd {type=NiagaraBool}
  User.bIsFrontEndPreview {type=NiagaraBool}
 ]
}
