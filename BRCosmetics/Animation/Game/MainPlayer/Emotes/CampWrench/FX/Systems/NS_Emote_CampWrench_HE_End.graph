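# Niagara system for the CampWrench emote's "HE_End" effect, a short burst of fading meshes, glow sprites and light.
# Flow spawns two differently oriented, growing meshes with jittered material parameters and flickering sprites, while FlowCont adds a longer-lived fading mesh.
# Looping_NonSphereElements partitions eight particles into mostly blue layers, using VisibilityTag to share one emitter across sprite, mesh and light renderers.
# That emitter takes scale, offset and time-of-day tint from system variables.
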NiagaraSystem /BRCosmetics/Animation/Game/MainPlayer/Emotes/CampWrench/FX/Systems/NS_Emote_CampWrench_HE_End
dependencies {
 usedDataInterfaces [&NiagaraDataInterfaceCurve]
 usedDynamicInputs []
 usedModules [
  &self:SystemSpawnScript.SystemScriptSource.SystemScriptGraph.NiagaraNodeAssignment_0.SetVariables_#1
  &SystemState
 ]
}
emitters {
 Flow {
  emitterUpdateScript [
   &EmitterState()
   &SpawnBurst_Instantaneous(Spawn Count=2, Use Spawn Probability=false)
  ]
  particleSpawnScript [
   &FN_SetSpawnGroup(Number of Spawn Groups=2)
   &InitializeParticle(
    Color=<rgba 1 1 1 0.3478>
    Lifetime=0.3
    Mesh Scale Mode=::NewEnumerator3
    Mesh Uniform Scale=&MakeCustomFloatFromBool(
     Bool=&SetBoolByIntComparison(A=$Particles.SpawnGroup, B=0, Comparison Type=::NewEnumerator4)
     True Float=0.4
     False Float=0.6
    )
    UsePositionOffset=false
    Write Material Random=true
   )
   &InitialMeshOrientation(Orientation Axis=<0 0 1>, Orientation Vector=<0 0 1>, Rotation=<0 0.25 0>, Use Orientation Vector=true, Use Rotation Vector=true) {module=InitialMeshOrientation001}
   &InitialMeshOrientation(
    Rotation=&MakeCustomVectorFromBool(
     Bool=&SetBoolByIntComparison(A=$Particles.SpawnGroup, B=0, Comparison Type=::NewEnumerator4)
     True Vector=<0 -0.05 -0.05>
     False Vector=<0.5 0.05 0.05>
    )
    Use Orientation Vector=true
    Use Rotation Vector=true
   )
  ]
  particleUpdateScript [
   &ParticleState()
   &ScaleMeshSize(Scale Factor=&Multiply_VectorByFloat(
     Vector=<1 1 1>
     Float=&FloatFromCurve(FloatCurve=DataInterface() {propertyValues {Curve {
         defaultValue=0
         keys [{arriveTangent=1.5809, interpMode=RCIM_Cubic, leaveTangent=1.5809, tangentMode=RCTM_User, time=0, value=1}, <at 1 1.5 RCIM_Cubic>]
        }}})
    ))
   /SetVariables(Particles.StaggerOffset:NiagaraFloat=&Waveform(
     "Waveform [1]"=::NewEnumerator12
     "[1] Phase"=&Add_Float(A=$Emitter.Age, B=&RandomRangeFloat(Minimum=0, Maximum=1))
     "[1] Frequency"=&RandomRangeFloat(Minimum=10, Maximum=15)
     "[1] Amplitude Min/Max"=<-0.02 0.02>
    )) {isSetParameters=true}
   &DynamicMaterialParameters(
    Index 0 Param 2=&Add_Float(A=$Particles.StaggerOffset, B=&RandomRangeFloat(Minimum=-0.2, Maximum=0.2))
    Module.Index0Float0Write=true
    Module.Index0Float1Write=true
    Module.Index0Float2Write=true
    Module.Index0Float3Write=true
   )
   /SetVariables(Particles.FlareDynamic:Vector4f=<0.5 0.5 0.5 1>) {isSetParameters=true}
   &ScaleColor(
    Scale Alpha=&Multiply_Float(
     A=2
     B=&FloatFromCurve(FloatCurve=DataInterface() {propertyValues {Curve {
         defaultValue=0
         keys [{interpMode=RCIM_Cubic, tangentMode=RCTM_User, time=0, value=1}, {interpMode=RCIM_Cubic, tangentMode=RCTM_User, time=1, value=0}]
        }}})
    )
    ScaleA=true
    ScaleRGB=true
   )
   &FN_Sprite_Attributes(
    Camera Offset=25
    Color=&MakeLinearColorFromVectorAndFloat(
     "Vector (RGB)"=<1 1 1>
     "Float (Alpha)"=&Multiply_Float(
      A=0.2
      B=&FloatFromCurve(FloatCurve=DataInterface() {propertyValues {Curve {
          defaultValue=0
          keys [{interpMode=RCIM_Cubic, tangentMode=RCTM_User, time=0, value=1}, {interpMode=RCIM_Cubic, tangentMode=RCTM_User, time=1, value=0}]
         }}})
     )
    )
    Create Camera Offset Attribute=true
    Create Color Attribute=true
    Create SpriteSize Attribute=true
    SpriteSize=&Multiply_Vector2D_ByFloat(
     Float=&Waveform("Waveform [1]"=::NewEnumerator12, "[1] Frequency"=10, "[1] Amplitude Min/Max"=<0 1>, "[1] Scale"=0.5, "[1] Bias"=0.75)
     Vector2D=&Multiply_Vector2D_ByFloat(
      Float=&FloatFromCurve(FloatCurve=DataInterface() {propertyValues {Curve {
          defaultValue=1
          keys [
           {arriveTangent=0.9994, interpMode=RCIM_Cubic, leaveTangent=0.9994, tangentMode=RCTM_User, time=0, value=1}
           {interpMode=RCIM_Cubic, tangentMode=RCTM_User, time=1, value=1.5}
          ]
         }}})
      Vector2D=<40 40>
     )
    )
    Write to Module Namespace=true
   )
  ]
  rendererClasses [&NiagaraMeshRendererProperties, &NiagaraSpriteRendererProperties]
  renderers [
   {rendererClass=&NiagaraMeshRendererProperties, rendererIndex=0}
   {rendererClass=&NiagaraSpriteRendererProperties, rendererIndex=1}
   {rendererClass=&NiagaraSpriteRendererProperties, rendererIndex=2}
  ]
 }
 FlowCont {
  emitterUpdateScript [
   &EmitterState()
   &SpawnBurst_Instantaneous(Use Spawn Probability=false, UseLoopCountLimit=false)
  ]
  particleSpawnScript [
   &InitializeParticle(
    Color Mode=::NewEnumerator0
    Lifetime Max=0.75
    Lifetime Min=0.5
    Lifetime Mode=::NewEnumerator1
    Mesh Scale Mode=::NewEnumerator3
    Mesh Uniform Scale=0.6
    Sprite Size Mode=::NewEnumerator3
    UsePositionOffset=false
    Write Material Random=true
   )
   &InitialMeshOrientation(Mesh Axis To Orient=<0 0 1>, Mesh Orientation Mode=::NewEnumerator2, Use Rotation Vector=false)
  ]
  particleUpdateScript [
   &ParticleState()
   &DynamicMaterialParameters(
    Index 0 Param 1=&RandomRangeFloat(Minimum=0, Maximum=1)
    Index 0 Param 2=&FloatFromCurve(
     FloatCurve=DataInterface() {propertyValues {Curve {defaultValue=0.5, keys [{tangentMode=RCTM_User, time=0, value=0}, {tangentMode=RCTM_User, time=1, value=1}]}}}
     Scale Curve=0.2
    )
    Module.Param0WriteEnabled=true
    Module.Param1WriteEnabled=true
    Module.Param2WriteEnabled=true
    Module.Param3WriteEnabled=false
   )
   &ScaleColor(
    Scale Alpha=&FloatFromCurve(
     FloatCurve=DataInterface() {propertyValues {Curve {
        defaultValue=1
        keys [
         {arriveTangent=9.0725, interpMode=RCIM_Cubic, leaveTangent=9.0725, tangentMode=RCTM_User, time=0, value=0}
         {arriveTangent=-2.5249, interpMode=RCIM_Cubic, leaveTangent=-2.5249, tangentMode=RCTM_User, time=0.1408, value=0.9847}
         {arriveTangent=-0.0405, interpMode=RCIM_Cubic, leaveTangent=-0.0405, tangentMode=RCTM_User, time=1, value=0}
        ]
       }}}
     Scale Curve=0.25
    )
    ScaleA=true
    ScaleRGB=true
   )
  ]
  rendererClasses [&NiagaraMeshRendererProperties]
  renderers [{rendererClass=&NiagaraMeshRendererProperties, rendererIndex=0}]
 }
 Looping_NonSphereElements {
  emitterUpdateScript [
   &EmitterState(
    Inactive Response=::NewEnumerator1
    Life Cycle Mode=::NewEnumerator1
    Loop Behavior=::NewEnumerator1
    Loop Duration Mode=::NewEnumerator1
    UseLoopDelay=false
   )
   &SpawnBurst_Instantaneous(Spawn Count=8, Use Spawn Probability=false, UseLoopCountLimit=false)
  ]
  particleSpawnScript [
   &PartitionParticles(Number Of Partitions=8)
   /SetVariables(Particles.VisibilityTag=$Particles.PartitionParticles.Partition) {isSetParameters=true}
   &InitializeParticle(
    Color=&MergeHelper_ColorSwitch(
     0=<rgba 0.0417 0.6007 1 1>
     1=<rgba 0.1529 0.4823 0.8116 0.3>
     2=<rgba 0.8521 5.731 10 1>
     3=<rgba 0.8 4.4 8 1>
     4=<rgba 0.8957 2.1232 3 1>
     5=<rgba 0.992 1.58 2 1>
     6=<rgba 2.0435 3.5217 5 1>
     7=<rgba 0.0729 0.6137 1 0.3>
     Index=$Particles.PartitionParticles.Partition
     Num=::NewEnumerator6
    )
    Lifetime=0.3
    Mesh Scale Mode=::NewEnumerator3
    Mesh Uniform Scale=&Multiply_Float(
     A=&MergeHelper_FloatSwitch(0=0, 1=0, 2=0, 3=0, 4=0, 5=0, 6=0.2, 7=0.262, Index=$Particles.PartitionParticles.Partition, Num=::NewEnumerator6)
     B=$System.GlobalScale
    )
    Position Offset=$System.PositionOffset
    Sprite Rotation Angle=&MergeHelper_FloatSwitch(
     0=&RandomRangeFloat(Minimum=0, Maximum=1)
     1=&RandomRangeFloat(Minimum=0, Maximum=1)
     2=&RandomRangeFloat(Minimum=0, Maximum=1)
     3=0
     4=&RandomRangeFloat(Minimum=0, Maximum=1)
     5=&RandomRangeFloat(Minimum=0, Maximum=1)
     Index=$Particles.PartitionParticles.Partition
     Num=::NewEnumerator4
    )
    Sprite Rotation Mode=::NewEnumerator1
    Sprite Size=&Multiply_Vector2D_ByFloat(
     Float=$System.GlobalScale
     Vector2D=&MergeHelper_Vector2Switch(
      0=<60 60>
      1=<36 36>
      2=<15 15>
      3=<25 25>
      4=<35 35>
      5=<38 38>
      6=<0 0>
      7=<0 0>
      Index=$Particles.PartitionParticles.Partition
      Num=::NewEnumerator6
     )
    )
    Sprite Size Mode=::NewEnumerator4
    UsePositionOffset=true
    Write Material Random=true
   )
   &CameraOffset(Camera Offset Amount=&MergeHelper_FloatSwitch(0=5, 1=15, 2=5, 3=15, 4=0, 5=-10, 6=0, 7=0, Index=$Particles.PartitionParticles.Partition, Num=::NewEnumerator6))
   /SetVariables(Particles.LightExponent=4) {isSetParameters=true}
   &InitialMeshOrientation(
    Mesh Orientation Mode=::NewEnumerator2
    Rotation=&RandomRangeVector(Minimum=<-0.05 0 0>, Maximum=<0.05 0 0>)
    Use Rotation Vector=true
   )
  ]
  particleUpdateScript [
   &ParticleState()
   &SpriteRotationRate(Rotation Rate=5)
   &ScaleSpriteSize(Uniform Curve Sprite Scale=DataInterface() {propertyValues {Curve {
       defaultValue=1
       keys [{interpMode=RCIM_Cubic, tangentMode=RCTM_User, time=0, value=0}, {interpMode=RCIM_Cubic, tangentMode=RCTM_User, time=1, value=1}]
      }}}) {module=ScaleSpriteSize001}
   &ScaleSpriteSize(
    Scale Sprite Size Mode=::NewEnumerator0
    Uniform Scale Factor=&Waveform(
     "Waveform [1]"=::NewEnumerator2
     "[1] Frequency"=0.5
     "[1] Compound Frequency"=1
     "[1] Amplitude Min/Max"=<0.9 1.2>
     "[1] Pulse Duration"=0.5
    )
   )
   &ScaleMeshSize(Scale Factor=&VectorFromFloat(Value=&FloatFromCurve(FloatCurve=DataInterface() {propertyValues {Curve {
         defaultValue=1
         keys [{interpMode=RCIM_Cubic, tangentMode=RCTM_User, time=0, value=0}, {interpMode=RCIM_Cubic, tangentMode=RCTM_User, time=1, value=1}]
        }}}))) {module=ScaleMeshSize001}
   &ScaleColor(
    Scale Alpha=&Multiply_Float(
     A=&FloatFromCurve(FloatCurve=DataInterface() {propertyValues {Curve {
         defaultValue=1
         keys [{interpMode=RCIM_Cubic, tangentMode=RCTM_User, time=0, value=1}, {interpMode=RCIM_Cubic, tangentMode=RCTM_User, time=1, value=0}]
        }}})
     B=&MakeFloatFromLinearColor(LinearColor=$System.TODColorMultiplier, Channel:ENiagaraFloatFromLinearColorOptions=::NewEnumerator3)
    )
    Scale RGB=&MakeVectorFromLinearColorRGB(Color=$System.TODColorMultiplier)
    ScaleA=true
    ScaleRGB=true
   )
   &DynamicMaterialParameters(
    Index 0 Param 1=1
    Index 0 Param 2=0.85
    Index 0 Param 4=1
    Module.Index0Float0Write=true
    Module.Index0Float1Write=true
    Module.Index0Float2Write=true
    Module.Index0Float3Write=true
   )
   /SetVariables(Particles.DynPar_BasicGlow:Vector4f=&MakeVector4(X=2, Y=0.2, Z=0, W=0)) {isSetParameters=true}
   /SetVariables(
    Particles.LightColor=&Multiply_LinearColorByFloat(
     Linear Color=$Particles.Color
     Float=&Waveform("Waveform [1]"=::NewEnumerator2, "[1] Frequency"=0.5, "[1] Compound Frequency"=1, "[1] Amplitude Min/Max"=<0 1.2>)
    )
    Particles.LightRadius=&Multiply_Float(
     A=&Multiply_Float(A=60, B=$System.GlobalScale)
     B=&FloatFromCurve(FloatCurve=DataInterface() {propertyValues {Curve {
         defaultValue=1
         keys [
          {arriveTangent=0.3976, interpMode=RCIM_Cubic, time=0, value=0}
          {arriveTangent=0.0057, interpMode=RCIM_Cubic, leaveTangent=0.0057, tangentMode=RCTM_User, time=0.28, value=0}
          {interpMode=RCIM_Cubic, leaveTangent=8.3333, time=0.3, value=1}
         ]
        }}})
    )
   ) {isSetParameters=true}
   /SetVariables(Particles.DynPar Caustic:Vector4f=&MakeVector4(X=3, Y=1, Z=0, W=0)) {isSetParameters=true}
   /SetVariables(Particles.DecalColor:LinearColor=&Multiply_LinearColorByFloat(
     Linear Color=$Particles.Color
     Float=&Waveform("Waveform [1]"=::NewEnumerator2, "[1] Frequency"=0.5, "[1] Compound Frequency"=1, "[1] Amplitude Min/Max"=<0.6 1.2>)
    )) {isSetParameters=true}
  ]
  rendererClasses [
   &NiagaraSpriteRendererProperties
   &NiagaraLightRendererProperties
   &NiagaraMeshRendererProperties
  ]
  renderers [
   {rendererClass=&NiagaraSpriteRendererProperties}
   {rendererClass=&NiagaraSpriteRendererProperties, rendererIndex=1}
   {rendererClass=&NiagaraLightRendererProperties, rendererIndex=2}
   {rendererClass=&NiagaraSpriteRendererProperties, rendererIndex=3}
   {rendererClass=&NiagaraSpriteRendererProperties, rendererIndex=4}
   {rendererClass=&NiagaraSpriteRendererProperties, rendererIndex=5}
   {rendererClass=&NiagaraMeshRendererProperties, rendererIndex=6}
   {rendererClass=&NiagaraMeshRendererProperties, rendererIndex=7}
  ]
 }
}
summary {
 systemName=NS_Emote_CampWrench_HE_End
 emitters [
  {emitterName=Flow, rendererClasses [&NiagaraMeshRendererProperties, &NiagaraSpriteRendererProperties]}
  {emitterName=FlowCont, rendererClasses [&NiagaraMeshRendererProperties]}
  {
   emitterName=Looping_NonSphereElements
   rendererClasses [
    &NiagaraSpriteRendererProperties
    &NiagaraLightRendererProperties
    &NiagaraMeshRendererProperties
   ]
  }
 ]
 userVariables [User.UseNearCameraFade {type=NiagaraBool}]
}
