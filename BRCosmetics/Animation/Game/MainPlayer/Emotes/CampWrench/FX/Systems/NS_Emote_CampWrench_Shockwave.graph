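# Niagara system for the CampWrench emote that plays a short blue shockwave burst.
# Energy_End expands and fades a mesh, while Glow pairs a growing camera-offset sprite with a pulsing particle light.
# Sparks bursts speed-stretched sprites from a ring with drag and curl noise, deriving soft-glow size and color for a second sprite renderer.
# System variables share offset and time-of-day tint; front-end and time-of-day modifiers adjust size and brightness.
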NiagaraSystem /BRCosmetics/Animation/Game/MainPlayer/Emotes/CampWrench/FX/Systems/NS_Emote_CampWrench_Shockwave
dependencies {
 usedDataInterfaces [&NiagaraDataInterfaceCurve, &NiagaraDataInterfaceColorCurve, &NiagaraDataInterfaceVectorCurve]
 usedDynamicInputs [&HueShiftLinearColor]
 usedModules [
  &self:SystemSpawnScript.SystemScriptSource.SystemScriptGraph.NiagaraNodeAssignment_1.SetVariables_#1
  &self:SystemSpawnScript.SystemScriptSource.SystemScriptGraph.NiagaraNodeAssignment_0.SetVariables_#2
  &SystemState
  &self:SystemSpawnScript.SystemScriptSource.SystemScriptGraph.NiagaraNodeAssignment_2.SetVariables_#3
 ]
}
emitters {
 Energy_End {
  emitterUpdateScript [
   &EmitterState(Inactive Response=::NewEnumerator1, Life Cycle Mode=::NewEnumerator1, Loop Behavior=::NewEnumerator1, Loop Duration=1, UseLoopDelay=false)
   &SpawnBurst_Instantaneous(Use Spawn Probability=false, UseLoopCountLimit=false)
  ]
  particleSpawnScript [
   &InitializeParticle(
    Color=&MultiplyLinearColor(
     Linear Color=<rgba 1 1 1 0.2>
     Scale Factor=&MakeVector4_FromLinearColor(LinearColor=$System.TODColorMultiplier)
    )
    Lifetime=0.35
    Mesh Scale=&Multiply_VectorByFloat(Vector=<-2 -4 -3.5>, Float=$System.GlobalScale)
    Mesh Scale Mode=::NewEnumerator4
    Position Offset=$System.PositionOffset
    UsePositionOffset=true
    Write Material Random=true
   )
   &InitialMeshOrientation(Mesh Axis To Orient=<0 0 0>, Mesh Orientation Mode=::NewEnumerator2, Rotation=<0 0 0.25>, Use Rotation Vector=true)
  ]
  particleUpdateScript [
   &ParticleState()
   &ScaleMeshSize(Scale Factor=&VectorFromCurve(VectorCurve=DataInterface() {propertyValues {
       XCurve {keys [<at 0 1.35 RCIM_Cubic>,<at 1 1.5 RCIM_Cubic>]}
       YCurve {keys [<at 0 1.35 RCIM_Cubic>,<at 1 1.6 RCIM_Cubic>]}
       ZCurve {keys [<at 0 1.35 RCIM_Cubic>,<at 1 1.6 RCIM_Cubic>]}
      }}))
   &ScaleColor(
    Scale Alpha=&FloatFromCurve(FloatCurve=DataInterface() {propertyValues {Curve {
        defaultValue=1
        keys [
         {arriveTangent=-2.392, interpMode=RCIM_Cubic, leaveTangent=-2.392, tangentMode=RCTM_User, time=0, value=1}
         {interpMode=RCIM_Cubic, tangentMode=RCTM_User, time=1, value=0}
        ]
       }}})
    Scale RGB=&MakeVectorFromLinearColorRGB(Color=&ColorFromCurve(ColorCurve=DataInterface() {propertyValues {
        AlphaCurve {keys []}
        BlueCurve {keys [<at 0.1405 1 RCIM_Cubic>,<at 0.4119 0.5449 RCIM_Cubic>]}
        GreenCurve {keys [<at 0.1405 0.6465 RCIM_Cubic>,<at 0.4119 0.2949 RCIM_Cubic>]}
        RedCurve {keys [<at 0.1405 0.1517 RCIM_Cubic>,<at 0.4119 0.0091 RCIM_Cubic>]}
       }}))
    ScaleA=true
    ScaleRGB=true
   )
  ]
  rendererClasses [&NiagaraMeshRendererProperties]
  renderers [{rendererClass=&NiagaraMeshRendererProperties, rendererIndex=0}]
 }
 Glow {
  emitterUpdateScript [
   &EmitterState(
    Inactive Response=::NewEnumerator1
    Life Cycle Mode=::NewEnumerator1
    Loop Behavior=::NewEnumerator1
    Loop Delay=1.6
    Loop Duration=0.5
    UseLoopDelay=false
   )
   &SpawnBurst_Instantaneous(Use Spawn Probability=false, UseLoopCountLimit=false)
  ]
  particleSpawnScript [
   &InitializeParticle(
    Color=&MultiplyLinearColor(
     Linear Color=<rgba 0.4783 1.9493 3 0.2>
     Scale Factor=&MakeVector4_FromLinearColor(LinearColor=$System.TODColorMultiplier)
    )
    Lifetime=0.35
    Sprite Size=<163.4946 255.2111>
    Sprite Size Mode=::NewEnumerator4
    UsePositionOffset=false
    Write Material Random=true
   )
   &SystemLocation(MaskSpawnGroup=false, Offset=<0 21.2695 120>)
   &CameraOffset(Camera Offset Amount=25)
  ]
  particleUpdateScript [
   &ParticleState()
   &ScaleSpriteSize(
    Uniform Curve Scale=&FN_FrontEndFloatModifier(Gameplay=1, Front End=0.8)
    Uniform Curve Sprite Scale=DataInterface() {propertyValues {Curve {keys [
        {arriveTangent=5.1246, interpMode=RCIM_Cubic, leaveTangent=5.1246, tangentMode=RCTM_User, time=0, value=0}
        {arriveTangent=1.7949, interpMode=RCIM_Cubic, leaveTangent=1.7949, tangentMode=RCTM_User, time=0.3, value=1.5}
        <at 1 2 RCIM_Cubic>
       ]}}}
   )
   &ScaleColor(
    Scale Alpha=&FloatFromCurve(
     FloatCurve=DataInterface() {propertyValues {Curve {keys [
         {arriveTangent=-2.7027, interpMode=RCIM_Cubic, leaveTangent=-2.7027, tangentMode=RCTM_User, time=0, value=1}
         {arriveTangent=-1, interpMode=RCIM_Cubic, leaveTangent=-1, time=0.312, value=0.3309}
         <at 1 0>
        ]}}}
     Scale Curve=&Multiply_Float(A=1, B=&FNTimeOfDayScalar(Night Scale=0.8, Morning Scale=1, Evening Scale=1))
    )
    ScaleA=true
    ScaleRGB=true
   )
   &FN_Light_Attributes(
    Color=&ColorFromCurve(ColorCurve=DataInterface() {propertyValues {
       AlphaCurve {keys []}
       BlueCurve {keys [<at 0.2351 4 RCIM_Cubic>,<at 0.9742 2 RCIM_Cubic>]}
       GreenCurve {keys [<at 0.2351 2.6264 RCIM_Cubic>,<at 0.9742 1.0631 RCIM_Cubic>]}
       RedCurve {keys [<at 0.2351 0.7033 RCIM_Cubic>,<at 0.9742 0.1262 RCIM_Cubic>]}
      }})
    Create Color Attribute=true
    Create Exponent Attribute=true
    Create Position Attribute=true
    Create Radius Attribute=true
    Enable Particle Light Enable=true
    Exponent=2
    Position=<0 20 90>
    Radius=&Multiply_Float(
     A=75
     B=&FloatFromCurve(FloatCurve=DataInterface() {propertyValues {Curve {keys [<at 0 1.8>,<at 0.2 2.05>,<at 0.5 2.2>,<at 1 1.8>]}}})
    )
    Scale Color=&FloatFromCurve(
     FloatCurve=DataInterface() {propertyValues {Curve {keys [
         {arriveTangent=5.626, interpMode=RCIM_Cubic, leaveTangent=5.626, tangentMode=RCTM_User, time=0, value=0.6}
         {arriveTangent=0.0895, interpMode=RCIM_Cubic, leaveTangent=0.0895, tangentMode=RCTM_User, time=0.1, value=1}
         {arriveTangent=-0.8586, interpMode=RCIM_Cubic, leaveTangent=-0.8586, tangentMode=RCTM_User, time=0.815, value=0.72}
         {arriveTangent=-4.4162, interpMode=RCIM_Cubic, leaveTangent=-4.4162, tangentMode=RCTM_User, time=1, value=0}
        ]}}}
     Scale Curve=&Multiply_Float(
      A=&FN_FrontEndFloatModifier(Gameplay=2.2908, Front End=5)
      B=&FNTimeOfDayScalar(Night Scale=0.5, Morning Scale=1, Evening Scale=0.8)
     )
    )
   )
  ]
  rendererClasses [&NiagaraSpriteRendererProperties, &NiagaraLightRendererProperties]
  renderers [
   {rendererClass=&NiagaraSpriteRendererProperties}
   {rendererClass=&NiagaraLightRendererProperties, rendererIndex=1}
  ]
 }
 Sparks {
  emitterUpdateScript [
   &EmitterState(Life Cycle Mode=::NewEnumerator1, Loop Behavior=::NewEnumerator1, Loop Duration=1, UseLoopDelay=false)
   &SpawnBurst_Instantaneous(
    Spawn Count=&RandomRangeInt(Minimum=100, Maximum=100)
    Use Spawn Probability=false
    UseLoopCountLimit=false
   )
  ]
  particleSpawnScript [
   &InitializeParticle(
    Color=&MultiplyLinearColor(
     Linear Color=<rgba 3 3 3 2>
     Scale Factor=&MakeVector4_FromLinearColor(LinearColor=$System.TODColorMultiplier)
    )
    Lifetime Max=0.8
    Lifetime Min=0.4
    Lifetime Mode=::NewEnumerator1
    Position Offset=$System.PositionOffset
    Sprite Size Mode=::NewEnumerator3
    Uniform Sprite Size=&Multiply_Float(
     A=&RandomRangeFloat(Minimum=2, Maximum=3)
     B=&FN_FrontEndFloatModifier(Gameplay=1, Front End=0.6)
    )
    UsePositionOffset=true
    Write Material Random=true
   )
   &ShapeLocation(
    Disc Coverage=0.7381
    MaskSpawnGroup=false
    Offset=<0 31.9255 0>
    Offset Mode=::NewEnumerator0
    Override Seed=false
    Ring Radius=120
    Rotation Angle=1
    Rotation Axis=<0 1 0.2187>
    Rotation Mode=::NewEnumerator1
    Shape Primitive=::NewEnumerator4
    UseOwnerScale=true
   )
   &AddVelocity(
    Cone Angle=35
    Cone Axis=<0 -1 0>
    Module.UseConeFalloff=false
    Module.UseVelDistribution=false
    Override Seed=false
    Velocity Mode=::NewEnumerator2
    Velocity Speed=&RandomRangeFloat(Minimum=500, Maximum=1000)
   )
   &AddVelocity(Module.Use Falloff Distance=false, Origin Offset=<0 840.0195 0>, Override Seed=false, Velocity Mode=::NewEnumerator1, Velocity Speed=50.61) {module=AddVelocity001}
  ]
  particleUpdateScript [
   &ParticleState()
   &ScaleColor(
    Scale Mode=::NewEnumerator1
    Scale RGBA=&MakeVector4_FromLinearColor(LinearColor=&ColorFromCurve(ColorCurve=DataInterface() {propertyValues {
        AlphaCurve {keys [<at 0.333 1>,<at 1 0>]}
        BlueCurve {keys [<at 0 1>,<at 0.166 5>,<at 0.333 4>,<at 0.75 3>]}
        GreenCurve {keys [<at 0 0.62>,<at 0.166 2.2>,<at 0.333 1.3333>,<at 0.75 0.5>]}
        RedCurve {keys [<at 0 0.24>,<at 0.166 0.2>,<at 0.333 0>,<at 0.75 0>]}
       }}))
    ScaleRGBA=true
   )
   &ScaleColor(
    Scale Alpha=&FNTimeOfDayScalar(Night Scale=0.5, Morning Scale=0.3, Evening Scale=0.3)
    Scale RGB=&VectorFromFloat(Value=&FloatFromCurve(FloatCurve=DataInterface() {propertyValues {Curve {keys [<at 0 20>,<at 0.2 20>,<at 1 5>]}}}))
    ScaleA=true
    ScaleRGB=true
   ) {module=ScaleColor001}
   &Drag(Drag=&RandomRangeFloat(Minimum=8, Maximum=15), Use Linear Drag=true, Use Rotational Drag=false)
   &CurlNoiseForce(Noise Strength=&RandomRangeFloat(Minimum=200, Maximum=400), Pan Noise=true)
   &GravityForce(Gravity=&RandomRangeVector(Minimum=<0 0 0>, Maximum=<0 0 -100>))
   &SolveForcesAndVelocity(Clamp Velocity=false, Limit Acceleration=false)
   &ScaleSpriteSizeBySpeed(Max Scale Factor=<1 5>, Min Scale Factor=<1 0.05>)
   &ScaleSpriteSize(Scale Factor=&MakeVector2D(
     X=&FloatFromCurve(FloatCurve=DataInterface() {propertyValues {Curve {keys [<at 0 0.5>,<at 0.1 1>,<at 1 0.5>]}}})
     Y=&FloatFromCurve(FloatCurve=DataInterface() {propertyValues {Curve {keys [<at 0 1>,<at 1 0>]}}})
    ))
   /SetVariables(
    Particles.SoftGlowColor:LinearColor=&Multiply_LinearColorByFloat(Linear Color=$Particles.Color, Float=0.02)
    Particles.SoftGlowSize:Vector2f=&Multiply_Vector2D_ByFloat(Float=3, Vector2D=$Particles.SpriteSize)
   ) {isSetParameters=true}
  ]
  rendererClasses [&NiagaraSpriteRendererProperties]
  renderers [
   {rendererClass=&NiagaraSpriteRendererProperties}
   {rendererClass=&NiagaraSpriteRendererProperties, rendererIndex=1}
  ]
 }
}
summary {
 systemName=NS_Emote_CampWrench_Shockwave
 emitters [
  {emitterName=Glow, rendererClasses [&NiagaraSpriteRendererProperties, &NiagaraLightRendererProperties]}
  {emitterName=Energy_End, rendererClasses [&NiagaraMeshRendererProperties]}
  {emitterName=Sparks, rendererClasses [&NiagaraSpriteRendererProperties]}
 ]
 userVariables [
  User.bIsFrontEnd {type=NiagaraBool}
  User.bIsFrontEndPreview {type=NiagaraBool}
 ]
}
